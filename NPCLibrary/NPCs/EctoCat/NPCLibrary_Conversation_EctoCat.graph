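# Conversation database for the EctoCat NPC, defining what it says when a player talks to it.
# The entry tag leads to one speech node with a default line and prioritized contextual messages.
# These cover a first-ever meeting, match conversation count checks, and reactions to specific player outfits via character ID checks.
# The speech node then branches into two link nodes that hand off to the shared NPC Quests and Basic Services conversations.
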ConversationDatabase /NPCLibrary/NPCs/EctoCat/NPCLibrary_Conversation_EctoCat
entry_points [{tag=Conversation.Entry.EctoCat, destinations [#1]}]
nodes [
 ConversationEntryPointNode_7 {
  id=#1
  class=ConversationEntryPointNode
  outputs [#2]
  properties {entryTag {tagName=Conversation.Entry.EctoCat}}
 }
 Speech_C_4 {
  id=#2
  class=Speech_C
  outputs [#3, #4]
  properties {
   nodeName=""
   generalConfig {
    defaultMessage="You Loopers are always so *serious*. Let loose!"
    contextualMessages [
     {
      message="It's time for more tricks and treats, bud."
      contextRequirements [{requirement {class=FortControllerRequirement_IsFirstEverConversationWithNPC}}]
      priority {curve {rowName=Tandem.Speech.Priority.FirstEverInteraction}}
     }
     {
      message="You Loopers are always so *serious*. Let loose!"
      contextRequirements [{requirement {
         class=FortControllerRequirement_CheckNPCMatchConversationCount
         properties {nPCUniqueIdTag {tagName=AISpawnerData.Type.Specialist.Heavy.EctoCat}}
        }}]
      priority {curve {rowName=None}}
     }
     {
      message=Boo. Hehe.
      contextRequirements [{requirement {
         class=FortControllerRequirement_CheckNPCMatchConversationCount
         properties {nPCUniqueIdTag {tagName=AISpawnerData.Type.Specialist.Heavy.EctoCat}}
        }}]
      priority {curve {rowName=None}}
     }
     {
      message="Treats are fine, but tricks? Bud, tricks are where it's at."
      contextRequirements [{requirement {
         class=FortControllerRequirement_CheckNPCMatchConversationCount
         properties {nPCUniqueIdTag {tagName=AISpawnerData.Type.Specialist.Heavy.EctoCat}}
        }}]
      priority {curve {rowName=None}}
     }
     {
      message="In a way, we both stick to the SHADOWs, wouldn't ya say?"
      contextRequirements [{requirement {
         class=FortControllerRequirement_HasCID
         properties {allowedCIds [&CID_693_Athena_Commando_M_BuffCat]}
        }}]
      priority {curve {rowName=Tandem.Speech.Priority.OutfitReaction}}
     }
     {
      message="Howdy howdy, cousin Meowdy."
      contextRequirements [{requirement {class=FortControllerRequirement_HasCID, properties {allowedCIds [/BRCosmetics/Athena/Items/Cosmetics/Characters/Character_MollyKit]}}}]
      priority {curve {rowName=Tandem.Speech.Priority.OutfitReaction}}
     }
     {
      message="You been tricking the locals like your unky taught ya, kid?"
      contextRequirements [{requirement {
         class=FortControllerRequirement_HasCID
         properties {allowedCIds [/BRCosmetics/Athena/Items/Cosmetics/Characters/CID_776_Athena_Commando_M_ProfessorPup]}
        }}]
      priority {curve {rowName=Tandem.Speech.Priority.OutfitReaction}}
     }
     {
      message=The skate parks in my cursed shadowy home realm would blow your mind.
      contextRequirements [{requirement {class=FortControllerRequirement_HasCID, properties {allowedCIds [&Character_ChillCat]}}}]
      priority {curve {rowName=Tandem.Speech.Priority.OutfitReaction}}
     }
     {
      message="Sometimes family is a shadow-uncle and his alt-universe niece in a mech. "
      contextRequirements [{requirement {class=FortControllerRequirement_HasCID, properties {allowedCIds [/BRCosmetics/Athena/Items/Cosmetics/Characters/Character_SpaceFeline]}}}]
      priority {curve {rowName=Tandem.Speech.Priority.OutfitReaction}}
     }
     {
      message=*sigh*... the tricks other-you and I get up to in my home realm...
      contextRequirements [{requirement {
         class=FortControllerRequirement_HasCID
         properties {allowedCIds [/BRCosmetics/Athena/Items/Cosmetics/Characters/CID_286_Athena_Commando_F_NeonCat]}
        }}]
      priority {curve {rowName=Tandem.Speech.Priority.OutfitReaction}}
     }
    ]
   }
  }
 }
 ConversationLinkNode_0 {id=#4, class=ConversationLinkNode, properties {remoteEntryTag {tagName=Conversation.Entry.BasicServices}}}
 ConversationLinkNode_1 {id=#3, class=ConversationLinkNode, properties {remoteEntryTag {tagName=Conversation.Entry.NPCLibrary.Quests}}}
]
